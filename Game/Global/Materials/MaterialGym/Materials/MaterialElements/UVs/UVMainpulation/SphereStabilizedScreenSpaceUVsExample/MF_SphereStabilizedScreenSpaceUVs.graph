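# Material function example in the Material Gym that derives sphere-stabilized screen-space UVs from view-space position and vertex normal.
# It crosses the normalized view-space position with the view-transformed vertex normal, then remaps the result from -1..1 to 0..1.
# As wired, the cross product's G component feeds both UV channels.
# An ObjectStabilizedScreenSpaceUVs call is present but does not feed the output.
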MaterialFunction /Game/Global/Materials/MaterialGym/Materials/MaterialElements/UVs/UVMainpulation/SphereStabilizedScreenSpaceUVsExample/MF_SphereStabilizedScreenSpaceUVs
graph {nodes [
  n1 = TransformPosition(None=WorldPosition().XYZ) {transformSourceType=TRANSFORMPOSSOURCE_World, transformType=TRANSFORMPOSSOURCE_View}
  n2 = VertexNormalWS()
  n3 = MaterialFunctionCall(Float3=CrossProduct(A=Normalize(VectorInput=@n1), B=Transform(None=@n2) {transformSourceType=TRANSFORMSOURCE_World, transformType=TRANSFORM_View})) {materialFunction=&BreakOutFloat3Components}
  n4 = ComponentMask(None=AppendVector(A=@n3.G, B=@n3.G)) {g=true, r=true}
  FunctionOutput(None=Add(A=Multiply(A=@n4) {constB=0.5}) {constB=0.5})
  MaterialFunctionCall!(Position, UVRadialScale) {materialFunction=/Game/Weapons/WeaponSkins/MaterialFunction/ObjectStabilizedScreenSpaceUVs}
 ]}
