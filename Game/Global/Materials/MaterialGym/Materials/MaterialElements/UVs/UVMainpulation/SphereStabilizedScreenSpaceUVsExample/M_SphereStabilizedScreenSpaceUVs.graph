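# Additive example material in MaterialGym demonstrating sphere-stabilized screen-space UVs, apparently for a bubble effect.
# A shared stabilization function feeds center-scaled UVs to both the bubble texture and a low-res noise texture.
# A panning planet texture drives sine distortion of the bubble lookup, whose boosted-saturation color gets a time-driven hue shift plus a faint Fresnel rim.
# Scrolling noise scales the world vertex normal, forming an animated displacement-style vector.
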Material /Game/Global/Materials/MaterialGym/Materials/MaterialElements/UVs/UVMainpulation/SphereStabilizedScreenSpaceUVsExample/M_SphereStabilizedScreenSpaceUVs
graph {
 properties {bAutomaticallySetUsageInEditor=false, blendMode=BLEND_Additive, usedShadingModels=Default Lit}
 nodes [
  n1 = Fresnel(ExponentIn=ScalarParameter() {defaultValue=5.096, parameterName=Param_3})
  n2 = Multiply(A=Time()) {constB=0.04}
  n3 = TextureCoordinate()
  n4 = ScalarParameter() {defaultValue=2, parameterName=Param_1}
  n5 = TextureSample(UVs=Panner(Coordinate=Multiply(A=@n3, B=@n4) {constB=2}) {speedX=0.02, speedY=0.03}) {texture=&/T_LemonCart_ProceduralPlanet_C_E_M}
  n6 = Time()
  n7 = Sine(None=Add(A=@n5.R, B=Multiply(A=@n6) {constB=0.03}))
  n8 = Constant() {r=0.004}
  n9 = MaterialFunctionCall() {materialFunction=&/MF_SphereStabilizedScreenSpaceUVs}
  n10 = Constant() {r=1.3}
  n11 = TextureSample(UVs=Add(
    A=Multiply(A=@n7, B=@n8)
    B=MaterialFunctionCall(UVs=@n9.Result, Texture Scale=@n10) {materialFunction=&ScaleUVsByCenter}.UVs
   )) {texture=&/T_MatGymbubble}
  n12 = Desaturation(None=@n11.RGB, Fraction=Constant() {r=-1})
  Multiply(A=Add(
    A=Multiply(A=@n1) {constB=0.1}
    B=MaterialFunctionCall(Hue Shift Percentage=@n2, Texture=@n12) {materialFunction=&HueShift}.Result
   ))
  n13 = Add(
   A=MaterialFunctionCall(UVs=@n9.Result, Texture Scale=Constant() {r=6}) {materialFunction=&ScaleUVsByCenter}.UVs
   B=Multiply(A=Time(), B=Constant() {r=0.05}) {constB=0.1}
  )
  Multiply(
   A=VertexNormalWS()
   B=LinearInterpolate(
    B=Constant() {r=4}
    Alpha=TextureSample(UVs=@n13) {samplerType=SAMPLERTYPE_Grayscale, texture=/Game/Global/Textures/Default/Noise/LowResBlurredNoiseNormalized}.R
   ) {constB=10}
  )
 ]
}
