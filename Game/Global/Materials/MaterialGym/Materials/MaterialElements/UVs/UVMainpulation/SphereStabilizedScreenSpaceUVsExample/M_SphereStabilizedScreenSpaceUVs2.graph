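# Additive material in the Material Gym's sphere-stabilized screen-space UV example folder, giving an animated, hue-cycling lava-lamp bubble look.
# Two noise-distorted, panning lava-lamp layers, one rotated and bump-offset, become heights, then normals via HeightToNormalSmooth.
# Their tangent-space XY samples a matcap; the summed result is oversaturated by negative desaturation, hue-shifted over time and combined with faint Fresnel.
# The minimum of both layers forms an Opacity reroute.
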Material /Game/Global/Materials/MaterialGym/Materials/MaterialElements/UVs/UVMainpulation/SphereStabilizedScreenSpaceUVsExample/M_SphereStabilizedScreenSpaceUVs2
graph {
 properties {bAutomaticallySetUsageInEditor=false, blendMode=BLEND_Additive, usedShadingModels=Default Lit}
 nodes [
  n1 = Fresnel(ExponentIn=Constant() {r=5.096})
  n2 = Multiply(A=Time()) {constB=0.04}
  n3 = MaterialFunctionCall(UVs=TextureCoordinate(), "Rotation Angle (0-1)"=Constant() {r=0.5}) {materialFunction=&CustomRotator}
  n4 = TextureSample(UVs=Panner(Coordinate=TextureCoordinate()) {speedY=0.3}) {samplerType=SAMPLERTYPE_Grayscale, texture=&LowResBlurredNoise}
  n5 = BumpOffset(Coordinate=Add(A=Panner(Coordinate=@n3.Rotated Values) {speedY=0.01}, B=Multiply(A=@n4.R) {constB=0.01}), Height=Constant() {r=-1.5})
  n6 = NamedRerouteDeclaration(None=Multiply(
    A=TextureSample(UVs=@n5) {samplerType=SAMPLERTYPE_LinearGrayscale, texture=/Game/Environments/Asteria/Sets/Lair/Materials/LavaLamp/T_LavaLamp_Background}.R
    B=Constant() {r=2.2738}
   ) {constB=2.5}) {name=Opacity_1, nodeColor=<rgba 1 0 0.5859 1>}
  n7 = Multiply(
   A=OneMinus(None=Saturate(None=@n6))
   B=ScalarParameter() {defaultValue=10.0902, group=Seared Scurf, parameterName=Background Normal Intensity, sortPriority=6}
  ) {constB=20}
  n8 = ComponentMask(None=Transform(None=MaterialFunctionCall!(Height=@n7, Absolute World Position, World Space Vertex Normals) {materialFunction=/Engine/Functions/Engine_MaterialFunctions03/Procedurals/HeightToNormalSmooth}.WorldSpaceNormal) {transformSourceType=TRANSFORMSOURCE_World, transformType=TRANSFORM_Tangent}) {g=true, r=true}
  n9 = TextureSampleParameter2D(UVs=Add(A=Multiply(A=@n8) {constB=0.5}) {constB=0.5}) {
   group=5. Textures
   parameterName=Matcap
   sortPriority=2
   texture=&/T_MatGymbubble
  }
  n10 = Panner(Coordinate=TextureCoordinate()) {speedY=0.04}
  n11 = TextureCoordinate()
  n12 = Multiply(A=TextureSample(UVs=Panner(Coordinate=@n11) {speedY=0.3}) {samplerType=SAMPLERTYPE_Grayscale, texture=&LowResBlurredNoise}.R) {constB=0.02}
  n13 = Multiply(
   A=TextureSample(UVs=Add(A=@n10, B=@n12)) {samplerType=SAMPLERTYPE_LinearGrayscale, texture=&T_LavaLamp_Background}.R
   B=Constant() {r=2.2738}
  ) {constB=2.5}
  n14 = Multiply(
   A=OneMinus(None=Saturate(None=@n13))
   B=ScalarParameter() {defaultValue=10.0902, group=Seared Scurf, parameterName=Background Normal Intensity, sortPriority=6}
  ) {constB=20}
  n15 = ComponentMask(None=Transform(None=MaterialFunctionCall(Height=@n14) {materialFunction=&HeightToNormalSmooth}.WorldSpaceNormal) {transformSourceType=TRANSFORMSOURCE_World, transformType=TRANSFORM_Tangent}) {g=true, r=true}
  n16 = TextureSampleParameter2D(UVs=Add(A=Multiply(A=@n15) {constB=0.5}) {constB=0.5}) {
   group=5. Textures
   parameterName=Matcap
   sortPriority=2
   texture=&T_MatGymbubble
  }
  n17 = Desaturation(None=Add(A=@n9.RGB, B=@n16.RGB), Fraction=Constant() {r=-1})
  Multiply(A=Add(
    A=Multiply(A=@n1) {constB=0.1}
    B=MaterialFunctionCall(Hue Shift Percentage=@n2, Texture=@n17) {materialFunction=&HueShift}.Result
   ))
  NamedRerouteDeclaration(None=Min(A=@n6, B=@n13)) {name=Opacity, nodeColor=<rgba 1 0 0.5859 1>}
  OneMinus(None=NamedRerouteUsage())
 ]
}
