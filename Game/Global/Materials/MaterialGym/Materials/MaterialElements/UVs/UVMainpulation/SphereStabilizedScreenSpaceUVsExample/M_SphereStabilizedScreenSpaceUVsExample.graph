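# An example material in MaterialGym that demonstrates the MF_SphereStabilizedScreenSpaceUVs material function, set up as additive Default Lit.
# The function's result drives the UVs of a texture sample of the UV editor's color grid, a test pattern for inspecting the mapping.
# The graph shows no link from the sample to a material output.
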Material /Game/Global/Materials/MaterialGym/Materials/MaterialElements/UVs/UVMainpulation/SphereStabilizedScreenSpaceUVsExample/M_SphereStabilizedScreenSpaceUVsExample
graph {
 properties {bAutomaticallySetUsageInEditor=false, blendMode=BLEND_Additive, usedShadingModels=Default Lit}
 nodes [
  n1 = MaterialFunctionCall() {materialFunction=&/MF_SphereStabilizedScreenSpaceUVs}
  TextureSample(UVs=@n1.Result) {texture=/UVEditor/Textures/UVEditorColorGrid_Color}
 ]
}
